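; Behavior tree decorator that decides whether an AI-controlled BuzzCool pawn may path around its goal processor, capping how many clones do so at once.
; It reads GoalActorProcessor from the blackboard, checks its class and interface, and gets its active clones through BuzzCoolProcessor_Interface.
; It counts clones tagged Ability.BuzzCool.Clone.Wander.Location.Skipping and fails once that count reaches the scalable-float limit SF_MaxPathingAllowed.
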(fn PerformConditionCheckAI
  (params (OwnerController :type AIController%20Object%20Reference) (ControlledPawn :type Pawn%20Object%20Reference))
  (bind _returnvalue (Utilities|Name|MakeLiteralName "GoalActorProcessor"))
  (bind _blackboard (Variables|/Script/AIModule.AIController|GetBlackboard OwnerController))
  (bind _returnvalue_2 (AI|Components|Blackboard|GetValueasObject _blackboard _returnvalue))
  (bind _pathingcounter (LocalVariables|GetPathingCounter))
  (Utilities|IsValid _returnvalue_2
    (:Is%20Valid
      (Utilities|IsA(soft) _returnvalue_2 "/BuzzCool/Gameplay/BP_BuzzCool_CloneSpawnActor_Processor.BP_BuzzCool_CloneSpawnActor_Processor_C"
        (:True
          (if (Utilities|DoesObjectImplementInterface _returnvalue_2 "/BuzzCool/Interfaces/BuzzCoolProcessor_Interface.BuzzCoolProcessor_Interface_C")
            (bind _activeclones (Class|BuzzCoolProcessorInterface|GetActiveClones(Message) _returnvalue_2))
            (LocalVariables|SetbCanPath true)
            (LocalVariables|SetPathingCounter 0)
            (bind _array_element (Utilities|Array|ForEachLoopwithBreak _activeclones)
              (:LoopBody
                (if (GameplayTags|HasMatchingGameplayTag _array_element "(TagName=\"Ability.BuzzCool.Clone.Wander.Location.Skipping\")")
                  (Math|Integer|IncrementInt _pathingcounter
                    (:%20%20
                      (if (>= _pathingcounter (ScalableFloat|GetValueasInteger (Variables|Internal|GetSF_MaxPathingAllowed)))
                        (LocalVariables|SetbCanPath false))))))
              (:Completed
                _))
            (return (LocalVariables|GetbCanPath))
            (else
              _)))
        (:False
          _))
      (return false))
    (:Is%20Not%20Valid
      _)))
