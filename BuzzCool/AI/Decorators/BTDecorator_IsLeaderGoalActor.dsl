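; Behavior tree decorator that passes when the AI's current goal actor is its leader.
; Its AI condition check reads the LeaderActor and GoalActor object keys from the owning controller's Blackboard and returns whether they are equal.
; Both key names are hardcoded literals, and the controlled pawn is unused.
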(fn PerformConditionCheckAI
  (params (OwnerController :type AIController%20Object%20Reference) (ControlledPawn :type Pawn%20Object%20Reference))
  (return (== (AI|Components|Blackboard|GetValueasObject (Variables|/Script/AIModule.AIController|GetBlackboard OwnerController) (Utilities|Name|MakeLiteralName "LeaderActor")) (AI|Components|Blackboard|GetValueasObject (Variables|/Script/AIModule.AIController|GetBlackboard OwnerController) (Utilities|Name|MakeLiteralName "GoalActor")))))
